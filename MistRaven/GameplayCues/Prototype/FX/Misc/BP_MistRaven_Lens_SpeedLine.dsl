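; BP_MistRaven_Lens_SpeedLine is a Niagara actor, filed as a prototype gameplay cue effect, that appears to provide a speed-line lens effect for MistRaven.
; On BeginPlay it casts its owner to a Fort player pawn and checks it for the MistRaven.AllAscendedEquipped gameplay tag.
; The result is written to the Niagara component's "Ascended" bool variable, letting the effect vary for Ascended players.
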(event EventBeginPlay
  (bind _asfort_player_pawn (Utilities|Casting|CastToFortPlayerPawn (Actor|GetOwner)))
  (bind _returnvalue (GameplayTags|HasMatchingGameplayTag(Message) _asfort_player_pawn "(TagName=\"MistRaven.AllAscendedEquipped\")"))
  (Niagara|SetNiagaraVariable(Bool) (Variables|NiagaraActor|GetNiagaraComponent) "Ascended" _returnvalue))

(fn ConstructionScript)
